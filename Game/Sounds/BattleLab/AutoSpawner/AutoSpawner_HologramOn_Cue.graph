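# Sound cue for the BattleLab auto-spawner's hologram turning on.
# A single wave player plays AutoSpawner_HologramOn_01 through a Modulator node.
# It routes to the Creative gadgets sound class, uses the small-toy attenuation asset at reduced volume, and overrides concurrency to cap simultaneous instances.
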SoundCue /Game/Sounds/BattleLab/AutoSpawner/AutoSpawner_HologramOn_Cue
root=Modulator_0
nodes [
 {id=Modulator_0, children [WavePlayer_1]}
 {id=WavePlayer_1, properties {soundWaveAssetPtr=&/AutoSpawner_HologramOn_01}}
]
properties {
 attenuationSettings=/Game/Sounds/Creative/Attenuations/Fort_Creative_SmallToy_Att
 bOverrideConcurrency=true
 maxDistance=3000
 soundClassObject=&Fort_Gadgets_Creative
 volumeMultiplier=0.5
 attenuationOverrides {falloffDistance=512, attenuationShapeExtents {x=256}}
 concurrencyOverrides {maxCount=2}
}
